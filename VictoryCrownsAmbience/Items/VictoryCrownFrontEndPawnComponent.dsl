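; Front-end pawn component that shows victory crown FX on an Athena player pawn in the lobby.
; BeginPlay caches the pawn and includes a front-end check and self-destruction.
; Playlist changes start a one-time HasCrown cosmetic stat watch and refresh the FX.
; UpdateForGCNL toggles a local looping gameplay cue based on playlist support and crown ownership, guarded against duplicates; EndPlay clears it.
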(event EventBeginPlay
  ;[1 1 1] Check if in front end
  (bind _asplayer_pawn_athena (Utilities|Casting|CastToPlayerPawn_Athena (Components|GetOwner))
    (:then
      (Variables|Default|SetAthenaPawn _asplayer_pawn_athena)
      (if (Reactive|Frontend|IsCosmeticObjectDisplayedinLobby 0)
        _
        (else
          _)))
    (:CastFailed
      _))
  (Components|DestroyComponent))

(event EventPlaylistChanged
  (if (Variables|FortPawnComponent_VictoryCrownsFrontEnd|GetbInitializeSuccess)
    (if (not (Variables|Default|GetWatchingStat))
      ;[1 1 1] Watch cosmetic stat HasCrown and update GCNL
      (bind (_statname _statvalue) (Cosmetics|WatchCosmeticStat (Variables|Default|GetAthenaPawn) "(TagName=\"StatManager.HasCrown\")")
        (:then
          (Variables|Default|SetWatchingStat true))
        (:OnStatChanged
          (Variables|Default|SetAccountHasCrown (> _statvalue 0)))))
    (CallFunction|UpdateforGCNL)))

(event EventEndPlay
  (params (EndPlayReason :type EEndPlayReason%20Enum :default "Destroyed"))
  (Variables|Default|SetAccountHasCrown false)
  (CallFunction|UpdateforGCNL))

(fn UpdateForGCNL
  (bind _returnvalue (Ability|GameplayCue|MakeGameplayCueParameters 0.0))
  ;[1 1 1] If should apply crown fx
  (LocalVariables|SetShouldApplyFX (and (Variables|FortPawnComponent_VictoryCrownsFrontEnd|GetbSelectedPlaylistSupportCrown) (Variables|Default|GetAccountHasCrown)))
  ;[1 1 1] Apply Crown FX
  (if (and (LocalVariables|GetShouldApplyFX) (not (Variables|Default|GetAppliedCrownFX)))
    (Projectile|AddGameplayCue%28Looping%29Local (Variables|Default|GetAthenaPawn) (Variables|Default|GetCrownFXCueTag) _returnvalue)
    (Variables|Default|SetAppliedCrownFX true)
    (elif (and (not (LocalVariables|GetShouldApplyFX)) (Variables|Default|GetAppliedCrownFX))
      ;[1 1 1] Remove Crown FX
      (Projectile|RemoveGameplayCue%28Looping%29Local (Variables|Default|GetAthenaPawn) (Variables|Default|GetCrownFXCueTag) _returnvalue)
      (Variables|Default|SetAppliedCrownFX false))))
